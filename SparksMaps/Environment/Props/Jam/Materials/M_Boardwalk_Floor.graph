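# Parameterized Default Lit, Nanite-enabled material for a boardwalk floor prop, defaulting to Pirate Cove ship textures.
# U Size and V Size scale the texture coordinates, which CustomRotator then turns by the Rotate parameter.
# The rotated UVs feed Diffuse, Specular mask and Normal texture parameters, so all three tile and rotate together.
# The Specular default comes from the ship's trims set, while Diffuse and Normal use the walls set.
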Material /SparksMaps/Environment/Props/Jam/Materials/M_Boardwalk_Floor
graph {
 properties {bUsedWithNanite=true, pixelDepthOffsetMode=PDOM_Legacy, usedShadingModels=Default Lit}
 nodes [
  n1 = ScalarParameter() {defaultValue=1, parameterName=U Size}
  n2 = ScalarParameter() {defaultValue=1, parameterName=V Size}
  n3 = MaterialFunctionCall(
   UVs=Multiply(A=TextureCoordinate(), B=AppendVector(A=@n1, B=@n2))
   Rotation Center=Constant()
   "Rotation Angle (0-1)"=ScalarParameter() {defaultValue=1.25, parameterName=Rotate}
  ) {materialFunction=&CustomRotator}
  TextureSampleParameter2D(UVs=@n3.Rotated Values) {parameterName=Diffuse, texture=/Game/Environments/Sets/Pirate_Cove/Pirate_Ship/Textures/T_Pirate_Ship_Walls_D}
  TextureSampleParameter2D(UVs=@n3.Rotated Values) {
   parameterName=Specular
   samplerType=SAMPLERTYPE_Masks
   texture=/Game/Environments/Sets/Pirate_Cove/Pirate_Ship/Textures/T_Pirate_Ship_Trims_S
  }
  TextureSampleParameter2D(UVs=@n3.Rotated Values) {parameterName=Normal, samplerType=SAMPLERTYPE_Normal, texture=/Game/Environments/Sets/Pirate_Cove/Pirate_Ship/Textures/T_Pirate_Ship_Walls_N}
 ]
}
